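; PCG example graph that generates a city from a jittered boundary shape and two jittered arteries.
; Arteries and recursive splits form districts, which are size-filtered and gridded into blocks.
; Artery, collector and local road footprints are unioned and subtracted from the boundary, leaving lots.
; Lots become texture-driven extruded buildings, randomly chosen plazas or greenery, which also fills undividable districts; ground meshes are added and only the boundary is output.
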(graph City_Circular_1km_2_arteries_NoAssetLibrary_Fast
  (params (Debug :type Bool) (LocalRoadWidth :type Float :default 1.9e+03) (CollectorRoadWidth :type Float :default 2.6e+03) (ArteryRoadWidth :type Float :default 3.4e+03) (BlockWidth :type Float :default 1.2e+04) (BlockLength :type Float :default 6e+03) (DistrictSizeLimit :type Float :default 4e+04) (BuildableLotsRatio :type Float :default 0.9))
  (bind Subgraph_0 (Subgraph :Graph /PCGPrimitives/Primitives/Create/Create_Spline_Shape.Create_Spline_Shape))
  (bind Subgraph_42 (Subgraph :Graph /PCGPrimitives/Primitives/Transform/Transform_Jitter.Transform_Jitter))
  ;[0.15 0.15 0.15 0.5] ARTERIES
  (bind Subgraph_2 (Subgraph :Graph /PCGPrimitives/Primitives/Create/Create_Spline_line.Create_Spline_line))
  (bind Subgraph_3 (Subgraph :Graph /PCGPrimitives/Primitives/Create/Create_Spline_line.Create_Spline_line))
  (bind Subgraph_1 (Subgraph :Graph /PCGPrimitives/Primitives/Transform/Transform_Resample.Transform_Resample))
  (bind Subgraph_35 (Subgraph :Graph /PCGPrimitives/Primitives/Transform/Transform_Jitter.Transform_Jitter))
  (bind Subgraph_40 (Subgraph :Graph /PCGPrimitives/Primitives/Transform/Transform_Resample.Transform_Resample))
  (bind Subgraph_41 (Subgraph :Graph /PCGPrimitives/Primitives/Transform/Transform_Jitter.Transform_Jitter))
  ;[0.15 0.15 0.15 0.5 50] DISTRICTS
  (bind Subgraph_4 (Subgraph :Graph /PCGPrimitives/Primitives/Subdivide/Subdivide_Areas_withSpline.Subdivide_Areas_withSpline))
  (bind Subgraph_6 (Subgraph :Graph /PCGPrimitives/Primitives/Subdivide/Subdivide_Areas_WithRecursiveSplit.Subdivide_Areas_WithRecursiveSplit))
  (bind Subgraph_7 (Subgraph :Graph /PCGPrimitives/Primitives/Filter/Filter_Areas_BySize.Filter_Areas_BySize :Size DistrictSizeLimit))
  (bind Subgraph_5 (Subgraph :Graph /PCGPrimitives/Primitives/Subdivide/Subdivide_Areas_withGrid.Subdivide_Areas_withGrid :GridLength BlockLength :GridWidth BlockWidth))
  ;[0.15 0.15 0.15 0.5 50] ROADS
  (bind Subgraph_28 (Subgraph :Graph /PCGPrimitives/Primitives/Assign/Assign_ShapeGrammarDefinition.Assign_ShapeGrammarDefinition :Size CollectorRoadWidth))
  (bind Subgraph_33 (Subgraph :Graph /PCGPrimitives/Primitives/Assign/Assign_ShapeGrammarDefinition.Assign_ShapeGrammarDefinition :Size LocalRoadWidth))
  (bind Subgraph_38 (Subgraph :Graph /PCGPrimitives/Primitives/Assign/Assign_ShapeGrammarDefinition.Assign_ShapeGrammarDefinition :Size ArteryRoadWidth))
  (bind Subgraph_27 (Subgraph :Graph /PCGPrimitives/Primitives/Compose/Compose_Areas_Union.Compose_Areas_Union))
  ;[0.15 0.15 0.15 0.5 50] LOTS
  (bind Subgraph_12 (Subgraph :Graph /PCGPrimitives/Primitives/Compose/Compose_Areas_Exclusion.Compose_Areas_Exclusion))
  (bind Subgraph_22 (Subgraph :Graph /PCGPrimitives/Primitives/Compose/Compose_Areas_Exclusion.Compose_Areas_Exclusion)) ; Here we are removing the Undividable districts from the lots,\nbecause we are preserving them to build forest later on.
  (bind Subgraph_30 (Subgraph :Graph /PCGPrimitives/Primitives/Filter/Filter_RandomChoice.Filter_RandomChoice :Ratio BuildableLotsRatio)) ; We exlcude randomly 10% of the lots, as we will later use them,\nTo build plazas.
  (bind Subgraph_13 (Subgraph :Graph /PCGPrimitives/Primitives/Filter/Filter_Areas_BySize.Filter_Areas_BySize))
  (bind Subgraph_11 (Subgraph :Graph /PCGPrimitives/Primitives/Filter/Filter_Areas_BySize.Filter_Areas_BySize))
  ;[0.15 0.15 0.15 0.5 50] BUILDINGS
  (bind Subgraph_14 (Subgraph :Graph /PCGPrimitives/Primitives/Subdivide/Subdivide_Areas_WithRecursiveSplit.Subdivide_Areas_WithRecursiveSplit))
  (bind Subgraph_15 (Subgraph :Graph /PCGPrimitives/Primitives/Filter/Filter_Areas_BySize.Filter_Areas_BySize))
  (bind Subgraph_29 (Subgraph :Graph /PCGPrimitives/Primitives/Write/Write_Attribute_fromTexture.Write_Attribute_fromTexture))
  (bind Subgraph_21 (Subgraph :Graph /PCGPrimitives/Primitives/Create/Create_Mesh_Extrude.Create_Mesh_Extrude))
  ;[0.15 0.15 0.15 0.5 50] Ground
  (bind Subgraph_25 (Subgraph :Graph /PCGPrimitives/Primitives/Extract/Extract_Segments.Extract_Segments))
  (bind Subgraph_17 (Subgraph :Graph /PCGPrimitives/Primitives/Assign/Assign_ShapeGrammarDefinition.Assign_ShapeGrammarDefinition))
  (bind Subgraph_26 (Subgraph :Graph /PCGPrimitives/Primitives/Transform/Transform_Areas_Expand.Transform_Areas_Expand))
  (bind Subgraph_18 (Subgraph :Graph /PCGPrimitives/Primitives/Create/Create_Mesh_Planar.Create_Mesh_Planar))
  ;[0.05488 0.151 0.03147 0.5 50] Greenery
  (bind Subgraph_20 (Subgraph :Graph /PCGPrimitives/Primitives/Compose/Compose_Areas_Intersection.Compose_Areas_Intersection))
  (bind Subgraph_23 (Subgraph :Graph /PCGPrimitives/Primitives/Fill/Fill_Areas_Randomly.Fill_Areas_Randomly))
  (bind Subgraph_24 (Subgraph :Graph /PCGPrimitives/Primitives/Spawn/Spawn_Assets.Spawn_Assets))
  ;[0.15 0.15 0.15 0.5 50] Plaza
  (bind Subgraph_45 (Subgraph :Graph /PCGPrimitives/Primitives/Transform/Transform_Areas_Contract.Transform_Areas_Contract))
  (bind Subgraph_31 (Subgraph :Graph /PCGPrimitives/Primitives/Fill/Fill_Areas_Uniformly.Fill_Areas_Uniformly))
  (bind Subgraph_32 (Subgraph :Graph /PCGPrimitives/Primitives/Filter/Filter_RandomChoice.Filter_RandomChoice))
  (bind Subgraph_44 (Subgraph :Graph /PCGPrimitives/Primitives/Spawn/Spawn_Assets.Spawn_Assets))
  (reroute Artery%201%20%20Out Subgraph_35)
  (reroute Artery%202%20Out Subgraph_41)
  (reroute Lots%20Out Subgraph_12)
  (reroute undividable%20Districts_1 Subgraph_7.Unselected)
  (connect Subgraph_2 Out Subgraph_1 PrimaryInput)
  (connect Subgraph_42 Out Subgraph_4 PrimaryInput)
  (connect Artery%201%20%20Out Out Subgraph_4 SecondaryInput)
  (connect Artery%202%20Out Out Subgraph_4 SecondaryInput)
  (connect Subgraph_7 Selected Subgraph_5 PrimaryInput)
  (connect Subgraph_4 Areas Subgraph_6 PrimaryInput)
  (connect Subgraph_6 Areas Subgraph_7 PrimaryInput)
  (connect Subgraph_13 Selected Subgraph_11 PrimaryInput)
  (connect Subgraph_42 Out Subgraph_12 PrimaryInput)
  (connect Subgraph_27 Out Subgraph_12 SecondaryInput)
  (connect Subgraph_30 Selected Subgraph_13 PrimaryInput)
  (connect Subgraph_11 Selected Subgraph_14 PrimaryInput)
  (connect Subgraph_14 Areas Subgraph_15 PrimaryInput)
  (connect Subgraph_25 Out Subgraph_17 PrimaryInput)
  (connect Subgraph_26 Out Subgraph_18 PrimaryInput)
  (connect Lots%20Out Out Subgraph_20 PrimaryInput)
  (connect undividable%20Districts_1 Out Subgraph_20 SecondaryInput)
  (connect Subgraph_29 Out Subgraph_21 PrimaryInput)
  (connect Subgraph_12 Out Subgraph_22 PrimaryInput)
  (connect undividable%20Districts_1 Out Subgraph_22 SecondaryInput)
  (connect Subgraph_11 Unselected Subgraph_23 PrimaryInput)
  (connect Subgraph_13 Unselected Subgraph_23 PrimaryInput)
  (connect Subgraph_20 Out Subgraph_23 PrimaryInput)
  (connect Subgraph_23 Out Subgraph_24 PrimaryInput)
  (connect Lots%20Out Out Subgraph_25 PrimaryInput)
  (connect Lots%20Out Out Subgraph_26 PrimaryInput)
  (connect Subgraph_28 Footprints Subgraph_27 PrimaryInput)
  (connect Subgraph_33 Footprints Subgraph_27 PrimaryInput)
  (connect Subgraph_38 Footprints Subgraph_27 PrimaryInput)
  (connect Subgraph_6 Subdivide%20Splines Subgraph_28 PrimaryInput)
  (connect Subgraph_15 Selected Subgraph_29 PrimaryInput)
  (connect Subgraph_22 Out Subgraph_30 PrimaryInput)
  (connect Subgraph_45 Out Subgraph_31 PrimaryInput)
  (connect Subgraph_31 Out Subgraph_32 PrimaryInput)
  (connect Subgraph_5 Subdividing_splines Subgraph_33 PrimaryInput)
  (connect Subgraph_1 Out Subgraph_35 PrimaryInput)
  (connect Subgraph_4 Subdividing%20Splines Subgraph_38 PrimaryInput)
  (connect Subgraph_42 Out Subgraph_38 PrimaryInput)
  (connect Subgraph_3 Out Subgraph_40 PrimaryInput)
  (connect Subgraph_40 Out Subgraph_41 PrimaryInput)
  (connect Subgraph_0 Out Subgraph_42 PrimaryInput)
  (connect Subgraph_32 Selected Subgraph_44 PrimaryInput)
  (connect Subgraph_30 Unselected Subgraph_45 PrimaryInput)
  (output Out Subgraph_42))
